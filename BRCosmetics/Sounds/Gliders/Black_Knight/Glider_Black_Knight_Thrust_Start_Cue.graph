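# Sound cue for the Black Knight glider's thrust start sound.
# A single wave player feeds the thrust-start wave through a root modulator left at default settings, with the cue's volume turned well down.
# It uses the shared glider open/close attenuation and concurrency, and is routed to the glider landing sound class.
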SoundCue /BRCosmetics/Sounds/Gliders/Black_Knight/Glider_Black_Knight_Thrust_Start_Cue
root=Modulator_1
nodes [
 {id=Modulator_1, children [WavePlayer_3]}
 {id=WavePlayer_3, properties {soundWaveAssetPtr=&/Glider_BlackKnight_Thrust_Start_V2}}
]
properties {
 attenuationSettings=/Game/Sounds/Attenuation/Fort_Player_Glider_OpenClose_Athena_Att
 maxDistance=10000
 soundClassObject=&Fort_Gliders_Land
 volumeMultiplier=0.15
 concurrencySet [&Athena_GliderOpenClose_Conc]
}
